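# Translucent, unlit, two-sided overlay material for a low-gravity effect, as its name indicates, usable on Nanite meshes.
# MF_WildEstate_ZPOverlay supplies one Emissive output that drives both emissive color and opacity.
# Those material attributes pass through MF_WE_WPOSway, which by its name adds world-position-offset sway, presumably to keep the overlay aligned with swaying meshes.
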Material /WildEstateTech/TechArt/Overlays/M_LowGrav_Overlay
graph {
 properties {
  bAutomaticallySetUsageInEditor=false
  bUseMaterialAttributes=true
  bUsedWithNanite=true
  blendMode=BLEND_Translucent
  shadingModel=MSM_Unlit
  twoSided=true
  usedShadingModels=Unlit
 }
 nodes [
  n1 = MaterialFunctionCall!(Material) {materialFunction=/WildEstateTech/Environment/BaseMaterials/MaterialFunctions/MF_WildEstate_ZPOverlay}
  MaterialFunctionCall!(In=MakeMaterialAttributes!(
    BaseColor
    Metallic
    Specular
    Roughness
    Anisotropy
    EmissiveColor=@n1.Emissive
    Opacity=@n1.Emissive
    OpacityMask
    Normal
    Tangent
    WorldPositionOffset
    SubsurfaceColor
    ClearCoat
    ClearCoatRoughness
    AmbientOcclusion
    Refraction
    CustomizedUVs_0
    CustomizedUVs_1
    CustomizedUVs_2
    CustomizedUVs_3
    CustomizedUVs_4
    CustomizedUVs_5
    CustomizedUVs_6
    CustomizedUVs_7
    PixelDepthOffset
    ShadingModel
    Displacement
   )) {materialFunction=/WildEstateTech/Environment/BaseMaterials/MaterialFunctions/MF_WE_WPOSway}
 ]
}
