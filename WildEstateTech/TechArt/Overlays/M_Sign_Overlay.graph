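# Translucent, unlit, two-sided overlay material for signs, built around cartoon-style line detection from scene depth.
# Center depth and four axis-offset samples feed combine4Samples, with offsets scaled by pixel depth, view resolution and resolution fraction.
# Depth fades drive divisor and sample bias, the latter scaled by object radius, with shading-path switches for mobile tuning.
# A separate vertex normal times outline_Thickness term suggests hull-style thickening.
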Material /WildEstateTech/TechArt/Overlays/M_Sign_Overlay
graph {
 properties {
  bAutomaticallySetUsageInEditor=false
  bUsedWithNanite=true
  blendMode=BLEND_Translucent
  shadingModel=MSM_Unlit
  twoSided=true
  usedShadingModels=Unlit
 }
 nodes [
  Constant3Vector()
  n1 = PixelDepth()
  n2 = ScalarParameter() {defaultValue=50, group=Cartoon, parameterName=CRTNO_InnerLine_FadeStart, sliderMax=200}
  n3 = Saturate(None=Divide(
    A=Subtract(A=@n1.R, B=@n2)
    B=ScalarParameter() {defaultValue=500, group=Cartoon, parameterName=CRTNO_InnerLine_FadeDistance, sliderMax=1000}
   ))
  n4 = LinearInterpolate(
   A=ScalarParameter() {group=Cartoon, parameterName=SampleBias_Min, sliderMax=16}
   B=ScalarParameter() {group=Cartoon, parameterName=SampleBias, sliderMax=16}
   Alpha=@n3
  )
  n5 = LinearInterpolate(
   A=ScalarParameter() {group=Cartoon, parameterName=MobileSampleBias_Min, sliderMax=1}
   B=ScalarParameter() {group=Cartoon, parameterName=Mobile_SampleBias, sliderMax=16}
   Alpha=@n3
  )
  n6 = ObjectRadius()
  n7 = ScalarParameter() {defaultValue=200, group=Cartoon, parameterName=ObjectScale, sliderMax=200}
  n8 = ComponentMask(None=ScreenPosition().ViewportUV) {g=true, r=true}
  n9 = Divide(A=ViewSize(), B=Constant() {r=320})
  n10 = Constant() {r=2}
  n11 = ScalarParameter() {
   defaultValue=2
   desc="! Dont touch this for BR or FE tweaks !\r\nThis is only ultra high resolutions for cinematic renderings"
   group=Cartoon
   parameterName=ResolutionCompensationAmount
   sliderMax=4
   sliderMin=2
   sortPriority=40
  }
  n12 = Clamp(None=Ceil(None=Subtract(A=@n9, B=@n10) {constB=0}), Max=@n11) {desc="if resolution is bigger than 960 =2", maxDefault=2, minDefault=1}
  n13 = Subtract(A=PixelDepth().R, B=ScalarParameter() {defaultValue=50, group=Cartoon, parameterName=CRTNO_Outline-InnerScaleStart, sliderMax=200})
  n14 = ScalarParameter() {defaultValue=500, group=Cartoon, parameterName=CRTNO_Outline-InnerScaleDistance, sliderMax=1000}
  n15 = LinearInterpolate(
   A=ScalarParameter() {defaultValue=2, group=Cartoon, parameterName=CRTNO_Outline-InnerSize1}
   B=ScalarParameter() {defaultValue=0.75, group=Cartoon, parameterName=CRTNO_Outline-InnerSize2}
   Alpha=Saturate(None=Divide(A=@n13, B=@n14))
  )
  n16 = Multiply(A=@n15, B=ScalarParameter() {defaultValue=1.5, group=Cartoon, parameterName=CRTNO_Outline-InnerSizeMobileMultiplier})
  n17 = Divide(A=Constant() {r=960}, B=ComponentMask(None=ViewSize()) {r=true})
  n18 = Max(A=ShadingPathSwitch(Default=@n15, Forward=@n16, Mobile=@n16), B=Saturate(None=Floor(None=@n17)) {desc=if resolution is smaller than 960H}) {desc="If res smaller than target, calmp min value to 1"}
  n19 = ViewProperty() {property=MEVP_ResolutionFraction}
  n20 = Multiply(A=SceneTexelSize(), B=Multiply(A=Multiply(A=@n12, B=@n18) {constA=2}, B=ComponentMask(None=@n19.Property) {r=true}))
  n21 = Constant()
  n22 = AppendVector(A=ComponentMask(None=@n20) {desc=x, r=true}, B=@n21)
  n23 = AppendVector(A=@n21, B=ComponentMask(None=@n20) {desc=y, g=true}) {desc=y+}
  n24 = Constant() {r=-1}
  n25 = Multiply(A=@n22, B=@n24)
  n26 = Multiply(A=@n24, B=@n23)
  n27 = ScalarParameter() {defaultValue=-1, group=Cartoon, parameterName=Divisor_Min, sliderMax=1, sliderMin=-1}
  n28 = ScalarParameter() {defaultValue=-16, group=Cartoon, parameterName=Divisor_max, sliderMax=16, sliderMin=-16}
  n29 = ScalarParameter() {defaultValue=-1, group=Cartoon, parameterName=MobileDivisor_Min, sliderMax=1, sliderMin=-1}
  n30 = ScalarParameter() {defaultValue=-16, group=Cartoon, parameterName=MobileDivisor_max, sliderMax=16, sliderMin=-16}
  MaterialFunctionCall!(
   Base=SceneDepth() {desc=SceneDepth works on mobile whereas SceneTexture doesn't}.R
   BaseSampleBias=Multiply(A=ShadingPathSwitch(Default=@n4, Mobile=@n5), B=Divide(A=@n6, B=@n7))
   Scale=ScalarParameter() {defaultValue=3.8208, group=Cartoon, parameterName=Scale, sliderMax=8}
   Sample1=SceneDepth(UVs=Add(A=@n8, B=@n22) {desc=x+}).R
   Sample2=SceneDepth(UVs=Add(A=@n8, B=@n23) {desc=y+}).R
   Sample3=SceneDepth(UVs=Add(A=@n8, B=@n25) {desc=x-}).R
   Sample4=SceneDepth(UVs=Add(A=@n8, B=@n26) {desc=y-}).R
   Divisor=ShadingPathSwitch(Default=LinearInterpolate(A=@n27, B=@n28, Alpha=@n3), Mobile=LinearInterpolate(A=@n29, B=@n30, Alpha=@n3))
   UseCenteredLines=Constant()
   UseRGB
  ) {materialFunction=/Game/Packages/jl_tests/combine4Samples}
  Multiply(A=VertexNormalWS(), B=ScalarParameter() {parameterName=outline_Thickness})
 ]
}
