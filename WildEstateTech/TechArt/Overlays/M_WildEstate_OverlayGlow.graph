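# Additive unlit overlay material for skeletal meshes that renders an exposure-compensated colored glow.
# Opacity combines a Fresnel term, swirl noise and depth fade, with noise built from panned, self-distorted storm textures on camera-aligned virtual-plane UVs.
# A vertical mask from pre-skinned bounds scales vertex offset along normals and camera push-back.
# That mask's bounds divisor subtracts Max from itself, yielding zero rather than mesh extent.
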Material /WildEstateTech/TechArt/Overlays/M_WildEstate_OverlayGlow
graph {
 properties {
  bAutomaticallySetUsageInEditor=false
  bUsedWithMeshDeformer=true
  bUsedWithSkeletalMesh=true
  blendMode=BLEND_Additive
  shadingModel=MSM_Unlit
  usedShadingModels=Unlit
 }
 nodes [
  n1 = ScalarParameter() {defaultValue=0.632, parameterName=Opacity}
  n2 = VertexNormalWS()
  n3 = Normalize(VectorInput=WorldPosition() {worldPositionShaderOffset=WPT_CameraRelativeNoOffsets}.XYZ)
  n4 = Power(Base=Multiply(A=DotProduct(A=@n2, B=@n3)) {constB=-1}, Exp=ScalarParameter() {defaultValue=7.0136, parameterName=FresnelEXP})
  n5 = Power(Base=Multiply(A=@n4) {constB=5}) {constExponent=6}
  n6 = NamedRerouteUsage()
  n7 = LinearInterpolate(A=@n4, B=Saturate(None=Multiply(A=@n5, B=@n6)), Alpha=ScalarParameter() {defaultValue=0.2, parameterName=NoiseAmount})
  DepthFade(
   Opacity=Multiply(A=Multiply(A=@n1, B=@n7), B=NamedRerouteUsage())
   FadeDistance=ScalarParameter() {defaultValue=10, parameterName=DepthFade_Distance}
  )
  EyeAdaptationInverse(LightValueInput=Multiply(A=VectorParameter() {parameterName=Glow, defaultValue=<rgba 0.5 0.0465 0.2951 1>}.RGB, B=NamedRerouteUsage()))
  n8 = ScalarParameter() {defaultValue=12613.3809, parameterName=CameraPushBack}
  n9 = VertexNormalWS()
  n10 = ScalarParameter() {defaultValue=2272.1421, parameterName=VertexOffset}
  n11 = PreSkinnedPosition()
  n12 = PreSkinnedLocalBounds()
  n13 = Subtract(A=ShaderStageSwitch(PixelShader=VertexInterpolator(VS=@n11).PS, VertexShader=@n11), B=@n12.Min)
  n14 = Subtract(A=@n12.Max, B=@n12.Max)
  n15 = ComponentMask(None=Add(A=Divide(A=@n13, B=@n14), B=ScalarParameter() {defaultValue=0.2199, group=Wild Estate Tech Art FX, parameterName=VertexMaskOffset}) {constB=0.2}) {b=true}
  n16 = Saturate(None=Power(Base=OneMinus(None=@n15), Exp=ScalarParameter() {defaultValue=3, group=Wild Estate Tech Art FX, parameterName=VertexMaskContrast}))
  Multiply(A=Add(A=Multiply(A=@n3, B=@n8), B=Multiply(A=@n9, B=@n10)), B=@n16)
  NamedRerouteDeclaration(None=@n16) {name=VerticalMask, nodeColor=<rgba 0 1 0.9766 1>}
  n17 = NamedRerouteUsage()
  n18 = TextureSample(UVs=Panner(Coordinate=Multiply(A=@n17) {constB=1.33}) {speedX=0.018, speedY=0.007}) {samplerType=SAMPLERTYPE_LinearGrayscale, texture=/Game/Global/Textures/Default/Flowmaps/T_StormSwirlMask}
  n19 = OneMinus(None=@n18.R)
  n20 = Add(A=Multiply(A=AppendVector(A=@n18.R, B=@n19)) {constB=0.05}, B=Panner(Coordinate=@n17) {speedX=-0.002, speedY=-0.005})
  NamedRerouteDeclaration(None=Multiply(
    A=Power(Base=@n18.R) {constExponent=0.85}
    B=TextureSample(UVs=@n20) {samplerType=SAMPLERTYPE_LinearGrayscale, texture=&T_StormSwirlMask}.R
   )) {name=Noise, nodeColor=<rgba 1 0 0.6797 1>}
  n21 = Constant3Vector() {constant=<rgba 0 0 -1 0>}
  n22 = Constant3Vector() {constant=<rgba 1 0 0 0>}
  NamedRerouteDeclaration(None=MaterialFunctionCall(
    Plane Normal Axis=Transform(None=@n21) {transformSourceType=TRANSFORMSOURCE_View}
    Plane X Axis=Transform(None=@n22) {transformSourceType=TRANSFORMSOURCE_View}
    Plane Center=ActorPositionWS()
    UV Scale=ScalarParameter() {defaultValue=200, group=Wild Estate Tech Art FX, parameterName=ZPNoise_Scale}
   ) {materialFunction=&VirtualPlaneCoordinates}.UVs) {name=CameraVector, nodeColor=<rgba 1 0.0833 0.3367 1>}
  NamedRerouteUsage()
 ]
}
